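# Material function for the GrainSpike quadruped companion's eyes, outputting emissive color, a pupil mask and remapped eye UVs.
# Each pupil is a procedural superellipse whose width and exponent morph between slit and rounder shapes, sharpened by CheapContrast and tinted with a two-color gradient.
# Eyeball UVs scale around parameterized left and right centers with radial falloff, then sample the body emissive texture.
# Blurred noise provides UV distortion for pupils and eyeballs.
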MaterialFunction /CosmeticCompanions/Assets/Quadruped/GrainSpike/Materials/MF_GrainSpike_pupil
graph {nodes [
  n1 = LinearInterpolate(A=NamedRerouteUsage(), B=NamedRerouteUsage(), Alpha=NamedRerouteUsage())
  n2 = TextureSampleParameter2D(UVs=@n1) {group=EyePupils, parameterName=EyeballEmissive, texture=/CosmeticCompanions/Assets/Quadruped/GrainSpike/Textures/T_GrainSpike_Body_E}
  n3 = NamedRerouteUsage()
  n4 = NamedRerouteUsage()
  n5 = Constant() {r=1}
  n6 = If(A=NamedRerouteUsage(), "A > B"=@n5, "A == B"=@n5, "A < B"=Constant()) {constB=0.01}
  n7 = LinearInterpolate(
   A=VectorParameter() {group=EyePupils, parameterName=EyePupilsColor_Top, defaultValue=<rgba 0 0.0275 0.09 1>}.RGB
   B=VectorParameter() {group=EyePupils, parameterName=EyePupilsColor_Bottom, defaultValue=<rgba 0 0.1027 0.35 1>}.RGB
   Alpha=Saturate(None=LinearInterpolate(A=@n3, B=@n4, Alpha=@n6))
  )
  n8 = LinearInterpolate(A=NamedRerouteUsage(), B=NamedRerouteUsage(), Alpha=NamedRerouteUsage())
  FunctionOutput(None=NamedRerouteDeclaration(None=LinearInterpolate(A=@n2.RGB, B=@n7, Alpha=@n8) {constB=0}) {name=OutPutEmissive, nodeColor=<rgba 0.5 1 0 1>}) {outputName=OurputColor}
  n9 = Abs(None=NamedRerouteUsage())
  n10 = NamedRerouteUsage()
  n11 = LinearInterpolate(Alpha=@n10) {constA=0.03}
  n12 = Constant() {r=0.8}
  n13 = LinearInterpolate(A=@n12, Alpha=@n10) {constB=2}
  n14 = Abs(None=NamedRerouteUsage())
  n15 = LinearInterpolate(A=ScalarParameter() {defaultValue=1.5, group=EyePupils, parameterName=PupilVerticalDilation}, Alpha=Power(Base=NamedRerouteUsage()))
  n16 = Add(A=Power(Base=Divide(A=@n9, B=@n11), Exp=@n13), B=Power(Base=Divide(A=@n14, B=@n15), Exp=@n13))
  n17 = ScalarParameter() {defaultValue=8, group=EyePupils, parameterName=Pupilcontrast}
  n18 = Saturate(None=MaterialFunctionCall(In=OneMinus(None=@n16), Contrast=@n17) {materialFunction=&CheapContrast}.Result)
  n19 = Abs(None=NamedRerouteUsage())
  n20 = NamedRerouteUsage()
  n21 = LinearInterpolate(Alpha=@n20) {constA=0.03}
  n22 = LinearInterpolate(A=@n12, Alpha=@n20) {constB=2}
  n23 = Abs(None=NamedRerouteUsage())
  n24 = Add(A=Power(Base=Divide(A=@n19, B=@n21), Exp=@n22), B=Power(Base=Divide(A=@n23, B=@n15), Exp=@n22))
  n25 = Saturate(None=MaterialFunctionCall(In=OneMinus(None=@n24), Contrast=@n17) {materialFunction=&CheapContrast}.Result)
  FunctionOutput(None=Max(
    A=NamedRerouteDeclaration(None=@n18) {name=pupil_L, nodeColor=<rgba 0.5938 1 0 1>}
    B=NamedRerouteDeclaration(None=@n25) {name=pupil_R, nodeColor=<rgba 0.5938 1 0 1>}
   )) {outputName=OurputEyePupilsMask}
  FunctionOutput(None=@n1) {outputName=OurputEyePupilsUV}
  n26 = NamedRerouteUsage()
  n27 = LinearInterpolate(
   A=ScalarParameter() {defaultValue=47, group=EyePupils, parameterName=Pupil size scale}
   B=LinearInterpolate(
    A=ScalarParameter() {defaultValue=23, group=EyePupils, parameterName=Pupil size Max}
    B=ScalarParameter() {defaultValue=20, group=EyePupils, parameterName=Pupil size Min}
    Alpha=NamedRerouteUsage()
   ) {constB=22}
   Alpha=NamedRerouteUsage()
  )
  MaterialExpressionMultiply_1 = Multiply(A=Multiply(A=Add(A=@n26) {constB=-0.5}) {constB=2}, B=@n27)
  n28 = Convert(None=@MaterialExpressionMultiply_1) {
   nodeName=Break Float2
   convertInputs [{
     type=Vector2
     expressionInput {
      expression=&self:MaterialExpressionMultiply_1
      inputName=None
      mask=0
      maskA=0
      maskB=0
      maskG=0
      maskR=0
      outputIndex=0
     }
    }]
   convertOutputs [{type=Scalar}, {type=Scalar}]
  }
  MaterialExpressionMultiply_3 = Multiply(A=Add(A=NamedRerouteUsage()) {constB=-0.5}, B=Multiply(A=@n27) {constB=-2}) {constB=2}
  n29 = Convert(None=@MaterialExpressionMultiply_3) {
   nodeName=Break Float2
   convertInputs [{
     type=Vector2
     expressionInput {
      expression=&self:MaterialExpressionMultiply_3
      inputName=None
      mask=0
      maskA=0
      maskB=0
      maskG=0
      maskR=0
      outputIndex=0
     }
    }]
   convertOutputs [{type=Scalar}, {type=Scalar}]
  }
  NamedRerouteDeclaration(None=Add(A=@n28, B=@n29)) {name=L_pupil_UVx, nodeColor=<rgba 0.1719 1 0 1>}
  NamedRerouteDeclaration(None=Add(A=@n28, B=@n29)) {name=L_pupil_UVy, nodeColor=<rgba 0.1719 1 0 1>}
  n30 = ScalarParameter() {defaultValue=0.4, group=EyePupils, parameterName=pupil_size, sliderMax=1.3, sliderMin=-0.2, sortPriority=0}
  NamedRerouteDeclaration(None=@n30) {name=Pupil size, nodeColor=<rgba 0 1 0.5078 1>}
  n31 = ScalarParameter() {group=EyePupils, parameterName=pupil_shape, sliderMax=1, sortPriority=0}
  NamedRerouteDeclaration(None=Saturate(None=OneMinus(None=@n31))) {name=Pupil Dilation, nodeColor=<rgba 0 0.5312 1 1>}
  n32 = NamedRerouteDeclaration(None=FunctionInput(Preview=TextureCoordinate()) {bUsePreviewValueAsDefault=true, inputName=UV_Input, inputType=FunctionInput_Vector2}) {name=UV_Input, nodeColor=<rgba 0.5938 0 1 1>}
  n33 = Constant2Vector() {g=-0.001, r=-0.001}
  n34 = TextureSample(UVs=Multiply(A=@n32, B=Constant() {r=5.7628}) {constB=6}) {samplerSource=SSM_Wrap_WorldGroupSettings, samplerType=SAMPLERTYPE_Grayscale, texture=&LowResBlurredNoise}
  n35 = ScalarParameter() {defaultValue=-0.004, group=EyePupils, parameterName=NoiseUVOffset_Pupils}
  NamedRerouteDeclaration(None=Add(A=Add(A=@n32, B=@n33), B=Multiply(A=@n34.R, B=@n35) {constB=0.002})) {name=NoiseUV_Pupils, nodeColor=<rgba 0 0.1562 1 1>}
  n36 = Constant2Vector() {g=-0.001, r=-0.001}
  n37 = ScalarParameter() {defaultValue=-0.004, group=EyePupils, parameterName=NoiseUVOffset_EyeBall}
  NamedRerouteDeclaration(None=Add(A=Add(A=@n32, B=@n36), B=Multiply(A=@n34.R, B=@n37) {constB=0.002})) {name=NoiseUV_EyeBall, nodeColor=<rgba 0 0.1562 1 1>}
  n38 = Constant() {r=0.2}
  n39 = TextureCoordinate()
  NamedRerouteDeclaration(None=SmoothStep(Min=@n38, Max=@n38, Value=ComponentMask(None=@n39) {g=true})) {name=L_R_Eye_mask, nodeColor=<rgba 1 0 0.7734 1>}
  n40 = Power(Base=NamedRerouteUsage())
  n41 = LinearInterpolate(
   A=ScalarParameter() {defaultValue=0.2, group=EyePupils, parameterName=L_R_ScaleSize_MinA, sliderMax=2}
   B=ScalarParameter() {defaultValue=0.5, group=EyePupils, parameterName=L_R_ScaleSize_MinB, sliderMax=2}
   Alpha=@n40
  )
  n42 = LinearInterpolate(
   A=ScalarParameter() {defaultValue=1, group=EyePupils, parameterName=L_R_ScaleSize_Max_B, sliderMax=2}
   B=ScalarParameter() {defaultValue=2, group=EyePupils, parameterName=L_R_ScaleSize_Max_A, sliderMax=2}
   Alpha=@n40
  )
  n43 = NamedRerouteUsage()
  n44 = Multiply(A=NamedRerouteUsage(), B=@n43)
  n45 = Constant()
  n46 = Saturate(None=NamedRerouteUsage())
  n47 = NamedRerouteDeclaration(None=LinearInterpolate(
    A=ScalarParameter() {defaultValue=2, group=EyePupils, parameterName=OuterRadiusMin}
    B=ScalarParameter() {defaultValue=0.95, group=EyePupils, parameterName=OuterRadiusMax}
    Alpha=Power(Base=@n46)
   )) {name=OuterRadius, nodeColor=<rgba 0 0.6484 1 1>}
  n48 = Length(None=AppendVector(A=NamedRerouteUsage(), B=NamedRerouteUsage())) {is=class MaterialExpressionLength}
  n49 = LinearInterpolate(
   B=NamedRerouteDeclaration(None=LinearInterpolate(A=@n41, B=@n42, Alpha=@n44) {constB=2}) {name=L_R_ScaleSize, nodeColor=<rgba 0 0.4609 1 1>}
   Alpha=OneMinus(None=SmoothStep(Min=@n45, Max=@n47, Value=@n48))
  ) {constA=1}
  n50 = Divide(A=NamedRerouteUsage(), B=@n49)
  n51 = NamedRerouteUsage()
  n52 = NamedRerouteUsage()
  NamedRerouteDeclaration(None=Subtract(A=Add(A=@n50, B=@n51), B=Divide(A=@n52, B=@n49) {desc=Scale the center offset by the scaling factor}) {desc=Re offset to maintain center}) {name=L_Eye_BG, nodeColor=<rgba 0 1 0.9062 1>}
  n53 = Constant()
  n54 = NamedRerouteUsage()
  n55 = Length(None=AppendVector(A=NamedRerouteUsage(), B=NamedRerouteUsage()))
  n56 = LinearInterpolate(B=NamedRerouteUsage(), Alpha=OneMinus(None=SmoothStep(Min=@n53, Max=@n54, Value=@n55))) {constA=1}
  n57 = Divide(A=NamedRerouteUsage(), B=@n56)
  n58 = NamedRerouteUsage()
  n59 = NamedRerouteUsage()
  NamedRerouteDeclaration(None=Subtract(A=Add(A=@n57, B=@n58), B=Divide(A=@n59, B=@n56) {desc=Scale the center offset by the scaling factor}) {desc=Re offset to maintain center}) {name=R_Eye_BG, nodeColor=<rgba 1 0.5859 0 1>}
  NamedRerouteDeclaration(None=AppendVector(
    A=ScalarParameter() {defaultValue=0.914, group=EyePupils, parameterName=L_Eye_UVOffsetU}
    B=ScalarParameter() {defaultValue=0.0839, group=EyePupils, parameterName=L_Eye_UVOffsetV}
   )) {name=L_eye_Position, nodeColor=<rgba 0.3125 1 0 1>}
  NamedRerouteDeclaration(None=AppendVector(
    A=ScalarParameter() {defaultValue=0.9135, group=EyePupils, parameterName=R_Eye_UVOffsetU}
    B=ScalarParameter() {defaultValue=0.248, group=EyePupils, parameterName=R_Eye_UVOffsetV}
   )) {name=R_eye_Position, nodeColor=<rgba 0.3125 1 0 1>}
  n60 = NamedRerouteUsage()
  MaterialExpressionMultiply_7 = Multiply(A=Multiply(A=Add(A=@n60) {constB=-0.5}) {constB=2}, B=@n27)
  n61 = Convert(None=@MaterialExpressionMultiply_7) {
   nodeName=Break Float2
   convertInputs [{
     type=Vector2
     expressionInput {
      expression=&self:MaterialExpressionMultiply_7
      inputName=None
      mask=0
      maskA=0
      maskB=0
      maskG=0
      maskR=0
      outputIndex=0
     }
    }]
   convertOutputs [{type=Scalar}, {type=Scalar}]
  }
  MaterialExpressionMultiply_8 = Multiply(A=Add(A=NamedRerouteUsage()) {constB=-0.5}, B=Multiply(A=@n27) {constB=-2}) {constB=2}
  n62 = Convert(None=@MaterialExpressionMultiply_8) {
   nodeName=Break Float2
   convertInputs [{
     type=Vector2
     expressionInput {
      expression=&self:MaterialExpressionMultiply_8
      inputName=None
      mask=0
      maskA=0
      maskB=0
      maskG=0
      maskR=0
      outputIndex=0
     }
    }]
   convertOutputs [{type=Scalar}, {type=Scalar}]
  }
  NamedRerouteDeclaration(None=Add(A=@n61, B=@n62)) {name=R_pupil_UVx, nodeColor=<rgba 0.1719 1 0 1>}
  NamedRerouteDeclaration(None=Add(A=@n61, B=@n62)) {name=R_pupil_UVy, nodeColor=<rgba 0.1719 1 0 1>}
  NamedRerouteUsage()
  Saturate(None=@n30)
  Saturate(None=@n43)
  TextureCoordinate()
  TextureCoordinate()
  TextureSample(UVs=Multiply(A=TextureCoordinate(), B=Constant() {r=8}) {constB=6}) {samplerSource=SSM_Wrap_WorldGroupSettings, samplerType=SAMPLERTYPE_Grayscale, texture=&LowResBlurredNoise}
 ]}
