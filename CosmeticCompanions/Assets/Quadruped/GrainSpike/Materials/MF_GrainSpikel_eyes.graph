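# Eye material function for the GrainSpike quadruped cosmetic companion, built on shared Fortnite eye textures.
# It fakes cornea refraction with an IOR-based view vector and displacement-map iris depth, offsetting UVs around left and right eye positions.
# It provides iris and sclera tinting, UV, sun or camera-light highlights, and MF_GrainSpike_pupil-driven emissive with optional facing falloff.
# Results set material attributes, with cornea debug views, then pass through MF_EmissiveFresnel.
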MaterialFunction /CosmeticCompanions/Assets/Quadruped/GrainSpike/Materials/MF_GrainSpikel_eyes
graph {nodes [
  Constant()
  Constant()
  MaterialExpressionConstant_0 = Constant() {r=1.376}
  MaterialExpressionPixelNormalWS_2 = PixelNormalWS()
  MaterialExpressionCameraVectorWS_3 = CameraVectorWS()
  Custom!(internalIoR=@MaterialExpressionConstant_0, normalW=@MaterialExpressionPixelNormalWS_2, cameraW=@MaterialExpressionCameraVectorWS_3) {
   code="float airIoR = 1.00029;\r\n\r\nfloat n = airIoR / internalIoR;\r\n\r\nfloat facing = dot(normalW, cameraW);\r\n\r\nfloat w = n * facing;\r\n\r\nfloat k = sqrt(1+(w-n)*(w+n));\r\n\r\nfloat3 t;\r\nt = (w - k)*normalW - n*cameraW;\r\nt = normalize(t);\r\nreturn -t;"
   desc=Refraction Direction
   inputs [
    {
     inputName=internalIoR
     input {expression=&self:MaterialExpressionConstant_0}
    }
    {
     inputName=normalW
     input {expression=&self:MaterialExpressionPixelNormalWS_2}
    }
    {
     inputName=cameraW
     input {expression=&self:MaterialExpressionCameraVectorWS_3}
    }
   ]
  }
  n1 = NamedRerouteUsage()
  n2 = NamedRerouteUsage()
  n3 = NamedRerouteUsage()
  n4 = SetMaterialAttributes!(
   MaterialAttributes
   BaseColor=StaticSwitchParameter(
    True=NamedRerouteUsage()
    False=StaticSwitchParameter(True=@n2, False=@n3) {group=01_Eye_Advanced, parameterName=Eye Visualize Cornea, sortPriority=208}
   ) {group=01_Eye_Advanced, parameterName=Eye Debug, sortPriority=210}
   Specular=NamedRerouteUsage()
   Metallic=NamedRerouteUsage()
   Roughness=NamedRerouteUsage()
   Normal=@n1
   ShadingModel=ShadingModel()
   EmissiveColor=Max(A=NamedRerouteUsage(), B=NamedRerouteUsage())
   Anisotropy
   Opacity
  ) {attributeSetTypes [
    #1
    #2
    #3
    #4
    #5
    #6
    #7
    #8
    #9
   ]}
  FunctionOutput(None=MaterialFunctionCall!(In=@n4) {materialFunction=/Game/Weapons/WeaponSkins/MaterialFunction/MF_EmissiveFresnel}.Result)
  n5 = NamedRerouteUsage()
  LinearInterpolate(A=@n5, B=Constant3Vector() {constant=<rgba 0 0 1 0>}, Alpha=NamedRerouteUsage())
  LinearInterpolate(A=@n5, B=Constant3Vector() {constant=<rgba 0 1 0 0>}, Alpha=NamedRerouteUsage())
  MaterialFunctionCall(Normal=@n1, Flatness=LinearInterpolate(Alpha=NamedRerouteUsage())) {materialFunction=&FlattenNormal}
  n6 = NamedRerouteUsage()
  n7 = StaticSwitchParameter(True=NamedRerouteUsage(), False=NamedRerouteUsage()) {group=01_Eye_Advanced, parameterName=Eye Use Sun Highlight, sortPriority=205}
  NamedRerouteDeclaration(None=Multiply(A=StaticSwitchParameter(True=@n6, False=@n7) {defaultValue=true, group=01_Eye_Advanced, parameterName=Eye Use UV Highlight, sortPriority=200}) {constB=0.6}) {name=EmissiveHighlight, nodeColor=<rgba 0.5938 0 1 1>}
  n8 = TextureCoordinate()
  n9 = NamedRerouteDeclaration(None=Subtract(A=NamedRerouteUsage(), B=NamedRerouteUsage())) {name=LeftEyeUVDelta, nodeColor=<rgba 0.9688 1 0 1>}
  NamedRerouteDeclaration(None=LinearInterpolate(A=@n8, B=Subtract(A=@n8, B=@n9), Alpha=NamedRerouteUsage())) {name=UVs, nodeColor=<rgba 0 1 0.2969 1>}
  n10 = NamedRerouteUsage()
  n11 = Constant2Vector() {g=0.5, r=0.5}
  n12 = NamedRerouteUsage()
  n13 = NamedRerouteUsage()
  NamedRerouteDeclaration(None=Power(Base=SphereMask(A=@n10, B=@n11, Radius=@n12, Hardness=@n13) {attenuationRadius=0.365, hardnessPercent=50}) {constExponent=0.5}) {name=eye displacement mask, nodeColor=<rgba 0.875 1 0 1>}
  n14 = VectorParameter() {group=01_Eye_Advanced, parameterName=Eye UV Highlight Color Tint, sortPriority=122, defaultValue=<rgba 1 1 1 1>}
  n15 = LinearInterpolate(
   A=ScalarParameter() {defaultValue=1, group=01_Eye_Advanced, parameterName=Eye UV Right Highlight Strength, sortPriority=124}
   B=ScalarParameter() {defaultValue=1, group=01_Eye_Advanced, parameterName=Eye UV Left Highlight Strength, sortPriority=124}
   Alpha=NamedRerouteUsage()
  )
  n16 = ScalarParameter() {defaultValue=1, group=01_Eye_Advanced, parameterName=Eye UV Highlight Strength, sortPriority=124}
  n17 = NamedRerouteUsage()
  n18 = VectorParameter() {group=01_Eye_Advanced, parameterName=Eye UV Highlight Pos, sortPriority=120, defaultValue=<rgba 0.92 0.12 0 0>}
  n19 = ComponentMask(None=@n18.RGB) {g=true, r=true}
  n20 = NamedRerouteUsage()
  n21 = ScalarParameter() {defaultValue=0.005, group=01_Eye_Advanced, parameterName=Eye UV Highlight Size, sortPriority=126}
  n22 = ScalarParameter() {defaultValue=0.6, group=01_Eye_Advanced, parameterName=Eye UV Highlight Hardness, sliderMax=1, sortPriority=127}
  n23 = Multiply(
   A=Multiply(A=@n14.RGB, B=StaticSwitch(True=@n15, False=@n16, Value=@n17))
   B=SphereMask(A=NamedRerouteUsage(), B=Multiply(A=@n19, B=@n20), Radius=@n21, Hardness=@n22)
  )
  n24 = VectorParameter() {group=01_Eye_Advanced, parameterName=Eye UV Highlight Pos 2, sortPriority=121, defaultValue=<rgba 0.878 0.12 0 0>}
  n25 = ComponentMask(None=@n24.RGB) {g=true, r=true}
  n26 = NamedRerouteUsage()
  n27 = LinearInterpolate(
   A=ScalarParameter() {defaultValue=1, group=01_Eye_Advanced, parameterName=Eye UV Right Highlight 2 Strength, sortPriority=128}
   B=ScalarParameter() {defaultValue=1, group=01_Eye_Advanced, parameterName=Eye UV Left Highlight 2 Strength, sortPriority=127}
   Alpha=NamedRerouteUsage()
  )
  n28 = ScalarParameter() {defaultValue=1, group=01_Eye_Advanced, parameterName=Eye UV Highlight 2 Strength, sortPriority=125}
  n29 = NamedRerouteUsage()
  n30 = VectorParameter() {group=01_Eye_Advanced, parameterName=Eye UV Highlight 2 Color Tint, sortPriority=123, defaultValue=<rgba 1 1 1 1>}
  n31 = Multiply(
   A=SphereMask(A=NamedRerouteUsage(), B=Multiply(A=@n25, B=@n26), Radius=@n21, Hardness=@n22)
   B=Multiply(A=StaticSwitch(True=@n27, False=@n28, Value=@n29), B=@n30.RGB)
  )
  NamedRerouteDeclaration(None=StaticSwitchParameter(True=Add(A=@n23, B=@n31), False=@n23) {group=01_Eye_Advanced, parameterName=Add Second UV Highlight, sortPriority=201}) {name=uvHighlight, nodeColor=<rgba 0.7813 0 1 1>}
  n32 = NamedRerouteUsage()
  n33 = DotProduct(A=CameraVectorWS(), B=NamedRerouteUsage())
  n34 = LinearInterpolate(Alpha=Multiply(A=@n33, B=@n33)) {constA=0.325}
  NamedRerouteDeclaration(None=Multiply(A=NamedRerouteUsage(), B=Divide(A=@n32, B=@n34) {desc=heightW / cosAlpha - scale distance to iris plane by facing ratio of camera and plane})) {name=Refraction Vector, nodeColor=<rgba 0.0781 1 0 1>}
  n35 = NamedRerouteUsage()
  n36 = NamedRerouteUsage()
  NamedRerouteDeclaration(None=LinearInterpolate(A=@n35, B=Add(A=@n35, B=@n36), Alpha=NamedRerouteUsage())) {name=EyeUVs, nodeColor=<rgba 0 1 0.6016 1>}
  n37 = CollectionParameter() {
   collection=&FortniteMaterialParameters
   parameterName=SunlightDirectionalVector
  }
  n38 = Normalize(VectorInput=Add(A=CameraVectorWS(), B=ComponentMask(None=@n37) {b=true, g=true, r=true}))
  n39 = Normalize(VectorInput=VertexNormalWS())
  NamedRerouteDeclaration(None=SmoothStep(Value=DotProduct(A=@n38, B=@n39) {desc=Blinn}) {constMin=0.995}) {name=SunDirectionHighlight, nodeColor=<rgba 1 0.75 0 1>}
  NamedRerouteDeclaration(None=ScalarParameter() {defaultValue=0.078, group=01_Eye_Advanced, parameterName="Eye Cornea Radius (UV)", sortPriority=112}) {name=EyeCorneaRadius, nodeColor=<rgba 0 1 0.2266 1>}
  NamedRerouteDeclaration(None=PixelNormalWS()) {name=_Normal, nodeColor=<rgba 0 0 1 1>}
  NamedRerouteDeclaration(None=Divide(
    A=Constant() {desc=airIoR, r=1.0003}
    B=ScalarParameter() {defaultValue=1.336, group=01_Eye_Advanced, parameterName=Eye Cornea IOR, sortPriority=102}
   ) {constA=1.0003}) {name=n, nodeColor=<rgba 0 1 0.2031 1>}
  n40 = NamedRerouteUsage()
  n41 = NamedRerouteUsage()
  NamedRerouteDeclaration(None=Multiply(A=NamedRerouteUsage(), B=DotProduct(A=@n40, B=@n41) {desc=facing}) {desc=w}) {name=w, nodeColor=<rgba 1 0.8297 0 1>}
  n42 = NamedRerouteUsage()
  n43 = NamedRerouteUsage()
  n44 = Multiply(A=Subtract(A=@n42, B=@n43), B=Add(A=@n42, B=@n43))
  NamedRerouteDeclaration(None=SquareRoot(None=Add(A=@n44)) {desc=k}) {name=k, nodeColor=<rgba 0.518 1 0 1>}
  NamedRerouteDeclaration(None=Transform(None=NamedRerouteUsage())) {name=EyeDirection, nodeColor=<rgba 1 0.3047 0 1>}
  n45 = NamedRerouteUsage()
  n46 = MaterialFunctionCall!(UV_Input=@n45) {materialFunction=&/MF_GrainSpike_pupil}
  n47 = StaticComponentMaskParameter(None=@n46.OurputColor) {defaultB=true, defaultG=true, defaultR=true, group=Emissive, parameterName=Emissive Component Mask}
  n48 = StaticSwitchParameter(True=Multiply(A=@n47, B=@n47), False=@n47) {group=Emissive, parameterName=LinearEmissiveMask}
  n49 = Fresnel(Normal=Transform(None=NamedRerouteUsage())) {baseReflectFraction=0, exponent=1}
  n50 = ScalarParameter() {
   defaultValue=1
   desc=Set to 1 to have fresnel emissive instead of camera facing emissive
   group=Emissive
   parameterName=CameraFacing_Inverted
   sliderMax=1
  }
  n51 = Round(None=@n50)
  n52 = Saturate(None=Abs(None=Subtract(A=@n49, B=@n51)))
  n53 = ScalarParameter() {defaultValue=1, group=Emissive, parameterName=Fresnel_Power, sliderMax=10}
  n54 = Multiply(A=@n48, B=Clamp(None=Power(Base=@n52, Exp=@n53)) {maxDefault=100000}) {desc=Emis*Facing}
  n55 = ScalarParameter() {defaultValue=1, group=Emissive, parameterName=emissive mult}
  n56 = Clamp(None=@n55, Max=ScalarParameter() {defaultValue=1, group=Emissive, parameterName=emissive mult_MobileMax})
  NamedRerouteDeclaration(None=Multiply(
    A=StaticSwitchParameter(True=@n54, False=@n48) {group=Emissive, parameterName=CameraFacing}
    B=ShadingPathSwitch(Default=@n55, Deferred=@n55, Mobile=@n56) {desc=Migrated from FeatureLevelSwitch}
   )) {name=EyeEmissive, nodeColor=<rgba 1 0 0.6328 1>}
  NamedRerouteDeclaration(None=LinearInterpolate(
    A=NamedRerouteUsage()
    B=NamedRerouteUsage()
    Alpha=ScalarParameter() {defaultValue=1, group=01_Eye_Advanced, parameterName=Eye Refraction Mix, sliderMax=1, sortPriority=106}
   )) {name="Eye UVs (normalized)", nodeColor=<rgba 0 1 0.5232 1>}
  n57 = Distance(A=NamedRerouteUsage(), B=Constant2Vector() {g=0.5, r=0.5})
  n58 = NamedRerouteUsage()
  n59 = SphereMask(A=@n57, B=@n58) {attenuationRadius=0.0075}
  n60 = SphereMask(A=@n57, B=Multiply(A=@n58, B=NamedRerouteUsage())) {attenuationRadius=0.0075}
  NamedRerouteDeclaration(None=LinearInterpolate(A=@n5, B=Constant3Vector() {constant=<rgba 1 0 0 0>}, Alpha=Add(A=@n59, B=@n60))) {name=VisualizeCornea, nodeColor=<rgba 1 0.3281 0 1>}
  n61 = NamedRerouteUsage()
  n62 = TextureSampleParameter2D(UVs=@n61) {
   group=00_Eye_Basic
   parameterName=SRM
   samplerType=SAMPLERTYPE_Masks
   sortPriority=1
   texture=/Game/Characters/Shared/FNEye/Textures/T_Eye_S_2022
  }
  NamedRerouteDeclaration(None=LinearInterpolate(A=@n62.R, Alpha=NamedRerouteUsage()) {constB=0}) {name=EyeSpecular, nodeColor=<rgba 0.6875 0 1 1>}
  NamedRerouteDeclaration(None=Constant() {r=0.365}) {name=IrisSize, nodeColor=<rgba 1 0 0.7969 1>}
  n63 = Multiply(A=NamedRerouteUsage(), B=NamedRerouteUsage())
  n64 = NamedRerouteUsage()
  n65 = Multiply(A=NamedRerouteUsage(), B=NamedRerouteUsage())
  n66 = Add(
   A=Subtract(A=Divide(A=@n63, B=@n64)) {constB=-0.5}
   B=Divide(A=Multiply(A=@n65) {constB=-1, desc=compensates for normal map not being 0-1}, B=@n64)
  )
  n67 = TextureSampleParameter2D(UVs=NamedRerouteDeclaration(None=Saturate(None=@n66)) {name=Iris UVs, nodeColor=<rgba 0.2422 1 0 1>}) {
   group=01_Eye_Advanced
   parameterName=Eye Iris Normal
   samplerType=SAMPLERTYPE_Normal
   sortPriority=100
   texture=/Game/Characters/Shared/FNEye/Textures/T_Eye_Sphere_N
  }
  NamedRerouteDeclaration(None=@n67.RGB) {name="eyeNormal (unflattened)", nodeColor=<rgba 0 1 0.1797 1>}
  NamedRerouteDeclaration(None=ScalarParameter() {defaultValue=0.2, group=01_Eye_Advanced, parameterName=Eye Cornea Mask Hardness, sortPriority=113}) {name=Eye Cornea Mask Hardness Reroute, nodeColor=<rgba 0.8047 0 1 1>}
  NamedRerouteDeclaration(None=Multiply(
    A=NamedRerouteDeclaration(None=@n62.B) {name=MetallicSample, nodeColor=<rgba 0 1 0.7188 1>}
    B=ScalarParameter() {defaultValue=0.7, group=00_Eye_Basic, parameterName=Eye Metallic Mult, sortPriority=11}
   ) {constB=0}) {name=EyeMetallic, nodeColor=<rgba 0.9922 0 1 1>}
  NamedRerouteDeclaration(None=Max(A=@n62.G, B=ScalarParameter() {group=00_Eye_Basic, parameterName=Eye Roughness Min, sortPriority=10})) {name=EyeRoughness, nodeColor=<rgba 0.6641 1 0 1>}
  n68 = Normalize(VectorInput=VertexNormalWS())
  n69 = VectorParameter() {group=01_Eye_Advanced, parameterName=Eye Camera Light Vector, sortPriority=36, defaultValue=<rgba -0.2 0.075 -0.5 0>}
  n70 = Transform(None=Normalize(VectorInput=@n69.RGB)) {transformSourceType=TRANSFORMSOURCE_View}
  NamedRerouteDeclaration(None=SmoothStep(Value=DotProduct(A=@n68, B=@n70)) {constMin=0.995}) {name=Camera Highlight, nodeColor=<rgba 1 0.75 0 1>}
  n71 = TextureSample(UVs=NamedRerouteUsage()) {
   constCoordinate=1
   desc=UVs not scaled by center as the base UVs are a ortho projection and this depth needs to stay in that space
   samplerType=SAMPLERTYPE_LinearColor
   texture=/Game/Characters/Shared/FNEye/Textures/T_EyeMidPlaneDisplacement
  }
  n72 = ScalarParameter() {defaultValue=0.4, group=01_Eye_Advanced, parameterName=Eye Iris UV Radius, sliderMax=0.5, sortPriority=107}
  n73 = TextureSample(UVs=MaterialFunctionCall(X=Add(A=@n72) {constB=0.5}, Y=Constant() {r=0.5}) {materialFunction=&MakeFloat2}.Result) {
   desc="Sample Displacement map as above, sampled at the Iris Radius"
   samplerType=SAMPLERTYPE_LinearColor
   texture=&T_EyeMidPlaneDisplacement
  }
  n74 = Subtract(A=@n71.RGB, B=@n73.R)
  n75 = Constant()
  NamedRerouteDeclaration(None=Multiply(A=Max(A=@n74, B=@n75), B=ScalarParameter() {defaultValue=0.5, group=01_Eye_Advanced, parameterName=Eye Iris Depth Scale, sortPriority=108}) {desc=Iris Depth}) {name=height, nodeColor=<rgba 0 1 0.4375 1>}
  n76 = NamedRerouteUsage()
  n77 = NamedRerouteUsage()
  n78 = TextureCoordinate()
  n79 = NamedRerouteUsage()
  NamedRerouteDeclaration(None=SphereMask(A=Multiply(A=@n76, B=@n77), B=Multiply(A=@n78, B=@n77) {constB=2}, Radius=Multiply(A=@n79) {constB=0.5}) {attenuationRadius=0.1}) {name=LeftEyeMask, nodeColor=<rgba 1 0 0.5391 1>}
  n80 = VectorParameter() {
   desc="uvPosition (UV0) of the left Eye. Top left is 0,0\r\nSet it to the same as the right eye if the same texture is used for both"
   group=01_Eye_Advanced
   parameterName="Eye Left UV Position (UV0)"
   sortPriority=111
   defaultValue=<rgba 0.94 0.35 0 1>
  }
  NamedRerouteDeclaration(None=ComponentMask(None=@n80.RGB) {g=true, r=true}) {name=Left Eye UVPosition, nodeColor=<rgba 1 0.375 0 1>}
  n81 = ScalarParameter() {
   defaultValue=2
   desc="Texture Width/Height (2048x1024 would be 2)"
   group=00_Eye_Basic
   parameterName=Eye Texture AspectRatio
   sortPriority=12
  }
  NamedRerouteDeclaration(None=MaterialFunctionCall(X=Constant() {r=1}, Y=Divide(B=@n81)) {materialFunction=&MakeFloat2}.Result) {name=AspectRatio, nodeColor=<rgba 1 0.4688 0 1>}
  NamedRerouteDeclaration(None=MaterialFunctionCall(BaseNormal=NamedRerouteUsage(), AdditionalNormal=NamedRerouteUsage()) {materialFunction=&BlendAngleCorrectedNormals}.Result) {name=EyeNormal, nodeColor=<rgba 0 1 0.9063 1>}
  n82 = TextureSampleParameter2D() {
   group=00_Eye_Basic
   parameterName=Normal
   samplerType=SAMPLERTYPE_Normal
   sortPriority=2
   texture=&FlatNormalBlue
  }
  NamedRerouteDeclaration(None=@n82.RGB) {name=Normal, nodeColor=<rgba 0.1435 0.3013 1 1>}
  n83 = VectorParameter() {group=01_Eye_Advanced, parameterName="Eye Right UV Position (UV0)", sortPriority=110, defaultValue=<rgba 0.94 0.122 0 0>}
  NamedRerouteDeclaration(None=ComponentMask(None=@n83.RGB) {g=true, r=true}) {name=RightEyeUVPosition, nodeColor=<rgba 0.7812 1 0 1>}
  NamedRerouteDeclaration(None=TextureSampleParameter2D(UVs=@n61) {group=00_Eye_Basic, parameterName=Diffuse, sortPriority=0, texture=/Game/Characters/Shared/FNEye/Textures/T_Eye_D}.RGB) {name=DiffuseMap, nodeColor=<rgba 0 0.8744 1 1>}
  NamedRerouteDeclaration(None=Multiply(A=NamedRerouteUsage(), B=NamedRerouteUsage())) {name=EyeCenter, nodeColor=<rgba 0 1 0.2734 1>}
  n84 = ScalarParameter() {defaultValue=0.02, group=01_Eye_Advanced, parameterName=Eye Refraction Mult, sortPriority=105}
  n85 = Constant2Vector() {g=1, r=-1}
  n86 = NamedRerouteUsage()
  n87 = MaterialFunctionCall() {materialFunction=&TangentBasis}
  n88 = NamedRerouteUsage()
  n89 = DotProduct(A=@n87.X U, B=@n88)
  n90 = Normalize(VectorInput=Subtract(A=@n87.X U, B=Multiply(A=@n89, B=@n88)))
  n91 = DotProduct(A=@n90, B=@n86)
  n92 = DotProduct(A=CrossProduct(A=@n90, B=@n88), B=@n86)
  n93 = Multiply(
   A=Multiply(
    A=Multiply(A=@n84, B=@n85)
    B=MaterialFunctionCall(X=@n91, Y=@n92) {materialFunction=&MakeFloat2}.Result
   )
   B=NamedRerouteUsage()
  )
  NamedRerouteDeclaration(None=AppendVector(A=@n93, B=Constant())) {name=debug, nodeColor=<rgba 1 0.4688 0 1>}
  NamedRerouteDeclaration(None=StaticBoolParameter() {group=01_Eye_Advanced, parameterName=Use Left Right Highlight Strength, sortPriority=202}) {name=UseLeftRightHighlight, nodeColor=<rgba 0 0.2031 1 1>}
  n94 = LinearInterpolate(A=NamedRerouteUsage(), B=NamedRerouteUsage(), Alpha=NamedRerouteUsage())
  n95 = NamedRerouteUsage()
  NamedRerouteDeclaration(None=StaticSwitchParameter(
    True=StaticSwitchParameter(True=@n94, False=@n95) {group=02_Eye_Tint, parameterName=UseAdvancedEyeTinting, sortPriority=20}
    False=NamedRerouteUsage()
   ) {group=02_Eye_Tint, parameterName=UseEyeColorTinting, sortPriority=-10}) {name=EyeColor_1, nodeColor=<rgba 0 0.3438 1 1>}
  n96 = TextureCoordinate()
  n97 = TextureSampleParameter2D!(
   UVs=MaterialFunctionCall!(
    UV Coord=NamedRerouteUsage()
    LenghtScale=ScalarParameter() {defaultValue=1, group=02_Eye_Tint, parameterName=IrisNoiseUVLength, sortPriority=42}
    RadialScale=ScalarParameter() {defaultValue=1, group=02_Eye_Tint, parameterName=IrisNoiseUVRadialScale, sortPriority=43}
   ) {materialFunction=/Game/Global/MaterialFunctions/TechArt/MF_PolarCoordinates}.Result
   "DDX(UVs)"=DDX(Value=@n96)
   "DDY(UVs)"=DDY(Value=@n96)
   Apply View MipBias
  ) {
   group=02_Eye_Tint
   mipValueMode=TMVM_Derivative
   parameterName=IrisNoise
   samplerType=SAMPLERTYPE_Masks
   sortPriority=41
   texture=/Game/Global/Textures/Default/Blanks/T_White
  }
  NamedRerouteDeclaration(None=@n97.R) {name=IrisNoise, nodeColor=<rgba 0 1 0.3438 1>}
  n98 = NamedRerouteUsage()
  n99 = OneMinus(None=Saturate(None=ScalarParameter() {group=02_Eye_Tint, parameterName=IrisDarkEdgeRadius, sliderMax=1, sortPriority=23}))
  NamedRerouteDeclaration(None=SphereMask(
    A=NamedRerouteUsage()
    B=Constant() {r=0.5}
    Radius=Multiply(A=@n98, B=@n99)
    Hardness=ScalarParameter() {defaultValue=1, group=02_Eye_Tint, parameterName=IrisDarkEdgeHardness, sliderMax=1, sortPriority=24}
   ) {attenuationRadius=0.4, hardnessPercent=80}) {name=OuterIrisMask, nodeColor=<rgba 1 0 0.3047 1>}
  n100 = NamedRerouteUsage()
  n101 = Constant() {r=0.5}
  n102 = NamedRerouteDeclaration(None=Multiply(A=ScalarParameter() {defaultValue=1, group=02_Eye_Tint, parameterName=IrisTintMask_Radius, sliderMax=1, sortPriority=2}) {constB=0.425}) {name=IrisRadius, nodeColor=<rgba 0 1 0.0156 1>}
  n103 = ScalarParameter() {defaultValue=0.8, group=02_Eye_Tint, parameterName=IrisTintMask_Hardness, sliderMax=1, sortPriority=3}
  n104 = StaticBoolParameter() {group=02_Eye_Tint, parameterName=UseMetalnessForIrisMask}
  NamedRerouteDeclaration(None=StaticSwitch(
    True=NamedRerouteUsage()
    False=SphereMask(A=@n100, B=@n101, Radius=@n102, Hardness=@n103) {attenuationRadius=0.4, hardnessPercent=80}
    Value=NamedRerouteDeclaration(None=@n104) {name=UseMetalnessForIrisMask, nodeColor=<rgba 1 0.3984 0 1>}
   )) {name=IrisMask, nodeColor=<rgba 0 0.625 1 1>}
  n105 = VectorParameter() {group=02_Eye_Tint, parameterName=ScleraTintColor, defaultValue=<rgba 1 1 1 1>}
  n106 = NamedRerouteUsage()
  n107 = ScalarParameter() {defaultValue=1, group=02_Eye_Tint, parameterName=ScleraTintStrength, sliderMax=1}
  NamedRerouteDeclaration(None=LinearInterpolate(A=NamedRerouteUsage(), B=Multiply(A=@n105.RGB, B=@n106), Alpha=Saturate(None=@n107))) {name=ScleraTint, nodeColor=<rgba 0.9453 1 0 1>}
  n108 = NamedRerouteDeclaration(None=Desaturation(None=NamedRerouteUsage(), Fraction=Constant() {r=1})) {name=DiffuseDesaturated, nodeColor=<rgba 0.7344 0 1 1>}
  n109 = VectorParameter() {group=02_Eye_Tint, parameterName=EyeTintColor2, sortPriority=22, defaultValue=<rgba 0.53 1 0.47 1>}
  n110 = VectorParameter() {group=02_Eye_Tint, parameterName=EyeTintColor, sortPriority=1, defaultValue=<rgba 0.53 1 0.47 1>}
  n111 = StaticSwitchParameter(
   True=@n109.RGB
   False=Multiply(A=@n110.RGB, B=ScalarParameter() {defaultValue=0.5, group=02_Eye_Tint, parameterName=IrisEdgeColorMultiplier, sortPriority=25})
  ) {group=02_Eye_Tint, parameterName=Use2ndIrisColor, sortPriority=21}
  n112 = NamedRerouteUsage()
  n113 = LinearInterpolate(B=NamedRerouteUsage(), Alpha=ScalarParameter() {group=02_Eye_Tint, parameterName=IrisNoiseStrength}) {constA=1}
  n114 = Saturate(None=StaticSwitchParameter(True=Multiply(A=@n112, B=@n113), False=NamedRerouteUsage()) {group=02_Eye_Tint, parameterName=UseIrisNoise, sortPriority=40})
  NamedRerouteDeclaration(None=Multiply(A=@n108, B=LinearInterpolate(A=@n111, B=@n110.RGB, Alpha=@n114))) {name=IrisTintAdvanced, nodeColor=<rgba 0 1 0.4141 1>}
  NamedRerouteDeclaration(None=LinearInterpolate(A=NamedRerouteUsage(), B=NamedRerouteUsage(), Alpha=NamedRerouteUsage())) {name=BasicEyeTint, nodeColor=<rgba 1 0.0938 0 1>}
  NamedRerouteDeclaration(None=Multiply(A=@n108, B=@n110.RGB)) {name=IrisTintBasic, nodeColor=<rgba 1 0 0.2578 1>}
  NamedRerouteDeclaration(None=@n46.OurputEyePupilsMask) {name=OurPutEyePupilsMask, nodeColor=<rgba 0.2422 0 1 1>}
  NamedRerouteDeclaration(None=@n46.OurputEyePupilsUV) {name=EyePupilsUV, nodeColor=<rgba 1 0 0.5859 1>}
  NamedRerouteDeclaration(None=Add(A=@n93, B=NamedRerouteUsage())) {name=Refracted UVs, nodeColor=<rgba 0.6406 1 0 1>}
  NamedRerouteDeclaration(None=CameraVectorWS()) {name=Camera Vector, nodeColor=<rgba 1 0 1 1>}
  n115 = NamedRerouteUsage()
  n116 = NamedRerouteUsage()
  n117 = Subtract(A=Multiply(A=Subtract(A=@n115, B=@n116), B=NamedRerouteUsage()), B=Multiply(A=NamedRerouteUsage(), B=NamedRerouteUsage())) {desc=t}
  NamedRerouteDeclaration(None=Multiply(A=Normalize(VectorInput=@n117)) {constB=-1}) {name=Eye Refraction Direction, nodeColor=<rgba 0 0.2882 1 1>}
  NamedRerouteDeclaration(None=MaterialFunctionCall(
    Normal=@n67.RGB
    Flatness=ScalarParameter() {defaultValue=0.35, group=01_Eye_Advanced, parameterName=Eye Iris Normal Flatten, sortPriority=101}
   ) {materialFunction=&FlattenNormal}.Result) {name="eyeNormal (flattened)", nodeColor=<rgba 0.4766 1 0 1>}
  NamedRerouteUsage()
  NamedRerouteUsage()
  PixelNormalWS()
  n118 = VectorParameter() {group=Emissive, parameterName=EmissiveUVs_RG_UpperLeftCorner_BA_LowerRightCorner, defaultValue=<rgba 0 0 1 1>}
  StaticSwitchParameter(
   True=MaterialFunctionCall!(
    UVs=@n45
    TextureIn=TextureObjectParameter() {group=Emissive, parameterName=Emissive, texture=&T_BlacksRGB}
    UpperLeftCorner=ComponentMask(None=@n118.RGB) {g=true, r=true}
    LowerRightCorner=AppendVector(A=@n118.B, B=@n118.B)
    ExportFloat4
   ) {materialFunction=/Engine/Functions/Engine_MaterialFunctions01/Texturing/TextureCropping}.CroppedMasked
   False=@n47
  ) {desc=CROP_EMIS, group=Emissive, parameterName=CroppedEmissive}
  TextureSampleParameter2D!(UVs=@n45, Apply View MipBias) {group=Emissive, parameterName=Emissive, texture=&T_BlacksRGB}
  VertexNormalWS()
 ]}
