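# Parent material for the eyes of GrainSpike, a quadruped cosmetic companion, set up for skeletal meshes with morph targets and mesh deformers.
# The eye look comes from MF_GrainSpikel_eyes, whose result passes through the MF_TechArtOverrides function.
# It uses Material Attributes and takes its shading model from the material expression, with Default Lit as the used model.
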Material /CosmeticCompanions/Assets/Quadruped/GrainSpike/Materials/M_GrainSpike_Eyes_Parent
graph {
 properties {
  bUseMaterialAttributes=true
  bUsedWithMeshDeformer=true
  bUsedWithMorphTargets=true
  bUsedWithSkeletalMesh=true
  pixelDepthOffsetMode=PDOM_Legacy
  shadingModel=MSM_FromMaterialExpression
  usedShadingModels=Default Lit
 }
 nodes [MaterialFunctionCall(In=MaterialFunctionCall() {materialFunction=&/MF_GrainSpikel_eyes}.Result) {materialFunction=&MF_TechArtOverrides}]
}
