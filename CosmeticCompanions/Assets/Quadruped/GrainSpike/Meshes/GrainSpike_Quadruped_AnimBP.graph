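# Animation Blueprint for the GrainSpike quadruped cosmetic companion, producing its pose by IK-retargeting from another mesh.
# The AnimGraph blends between an emote retargeter and the standard retargeter using the IsEmote flag.
# Each update sets IsEmote when the owning actor is a FortCompanionMannequin or the IsEmote curve equals 1, otherwise clears it.
# A Copy Pose From Mesh node is present but not connected to the output.
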AnimBlueprint /CosmeticCompanions/Assets/Quadruped/GrainSpike/Meshes/GrainSpike_Quadruped_AnimBP
target_skeleton=&/GrainSpike_Quadruped_Skeleton
graphs [
 AnimGraph {nodes [
   Copy Pose From Mesh() {bCopyCurves=true, bUseAttachedParent=true}
   Output Pose(Result=Blend Poses by bool(
     BlendPose_0=Retarget Pose From Mesh!() -> (Pose:Pose Link Structure) {is=title AnimGraphNode_RetargetPoseFromMesh, iKRetargeterAsset=&/RTG_GrainSpike_Emote}
     BlendPose_1=Retarget Pose From Mesh() {iKRetargeterAsset=&/RTG_GrainSpike}
     BlendTime_0=0.3
     bActiveValue=Get IsEmote() -> (IsEmote:bool)
    ))
  ]}
 EventGraph {nodes [
   Event BlueprintUpdateAnimation()
   n1 = GetOwningActor()
   n2 = GetCurveValue(CurveName=IsEmote)
   > n3 = Branch(Condition=OR Boolean(
     A=IsA!(Object:Object Reference=@n1, ObjectClass:Object Class Reference=/Script/CosmeticCompanionsRuntime.FortCompanionMannequin) -> (ReturnValue:bool)
     B="Equal (Float)"(A=@n2, B=1)
    ))
   > Set IsEmote(IsEmote:bool=true) -> (Output_Get:bool)
   Set IsEmote(execute=@n3.else, IsEmote=false)
   TryGetPawnOwner()
  ]}
]
variables [IsEmote {type=boolean}]
